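# Movie Render Queue burn-in widget, apparently for face ROM tech-anim renders, that overlays render info on output frames.
# On each output frame start, a Sequence calls per-region update functions that set job name, initialization date, author and root frame number.
# The build label combines the engine changelist with a branch name parsed after the engine version's last "+".
# Bottom-center is empty; bottom-right and user-comment functions are unwired, displaying nothing.
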WidgetBlueprint /Game/NeverCook/ArtTools/Tech_Anim/FaceRom_BurnIn
graphs [
 EventGraph {nodes [
   n1 = Event OnOutputFrameStarted!() -> (OutputDelegate, then, ForPipeline:Movie Pipeline Object Reference) {is=title K2Node_Event}
   > n2 = Sequence!(execute) -> (then_0, then_1, then_2:Exec, then_3:Exec, then_4:Exec, then_5:Exec)
   > UpdateTopLeftText!(execute, self, ForPipeline:Movie Pipeline Object Reference=@n1.ForPipeline) -> (then)
   UpdateTopRightText!(execute=@n2.then_1, self, ForPipeline:Movie Pipeline Object Reference=@n1.ForPipeline) -> (then)
   UpdateBottomLeftText!(execute=@n2.then_2, self, ForPipeline:Movie Pipeline Object Reference=@n1.ForPipeline) -> (then)
   UpdateBottomCenter!(execute=@n2.then_3, self, ForPipeline:Movie Pipeline Object Reference=@n1.ForPipeline) -> (then)
   UpdateBottomRightText!(execute=@n2.then_4, self, ForPipeline:Movie Pipeline Object Reference=@n1.ForPipeline) -> (then)
   UpdateUserCommentText!(execute=@n2.then_5, self, For Pipeline:Movie Pipeline Object Reference=@n1.ForPipeline) -> (then)
  ]}
 UpdateBottomCenter {nodes [UpdateBottomCenter!() -> (then, ForPipeline:Movie Pipeline Object Reference) {is=title K2Node_FunctionEntry}]}
 UpdateBottomLeftText {nodes [
   n1 = UpdateBottomLeftText!() -> (then, ForPipeline:Movie Pipeline Object Reference) {is=title K2Node_FunctionEntry}
   n2 = Break Frame Number!(FrameNumber:"Frame Number Structure (by ref)"=GetRootFrameNumber!(InMoviePipeline:Movie Pipeline Object Reference=@n1) -> (ReturnValue:Frame Number Structure)) -> (Value:int) {is=title K2Node_BreakStruct}
   > "SetText (Text)"(self=Get BottomLeft() -> (BottomLeft:Text Object Reference), InText=Format Text(Format="To Text (Integer)"(Value=@n2)))
   "To Text (String)"(InString=Timecode To String!(
     InTimecode:"Timecode Structure (by ref)"=GetRootTimecode!(InMoviePipeline:Movie Pipeline Object Reference=@n1) -> (ReturnValue:Timecode Structure)
     bForceSignDisplay:bool=false
    ) -> (ReturnValue:String))
  ]}
 UpdateBottomRightText {nodes [
   n1 = UpdateBottomRightText!() -> (then, ForPipeline:Movie Pipeline Object Reference) {is=title K2Node_FunctionEntry}
   Format Text!(
    Format="LOCTEXT(\"{ShotName}\")"
    ShotName:Text=GetCurrentSegmentName!(InMoviePipeline:Movie Pipeline Object Reference=@n1) -> (OutOuterName:Text, OutInnerName:Text)
   ) -> (Result)
   n2 = GetCurrentShotTimecode!(InMoviePipeline:Movie Pipeline Object Reference=@n1) -> (ReturnValue:Timecode Structure)
   Format Text!(
    Format="LOCTEXT(\"Shot:   {ShotTimecode} ({ShotFrameNumber}f)\")"
    ShotTimecode:Text="To Text (String)"(InString=Timecode To String(InTimecode=@n2, bForceSignDisplay=false))
    ShotFrameNumber:int=Break Frame Number(FrameNumber=GetCurrentShotFrameNumber!(InMoviePipeline:Movie Pipeline Object Reference=@n1) -> (ReturnValue:Frame Number Structure))
   ) -> (Result)
  ]}
 UpdateTopLeftText {nodes [
   n1 = UpdateTopLeftText!() -> (then, ForPipeline:Movie Pipeline Object Reference) {is=title K2Node_FunctionEntry}
   > "SetText (Text)"(
    self=Get Top_01() -> (Top_01:Text Object Reference)
    InText=Format Text!(
     Format="LOCTEXT(\"{SequenceName}\")"
     SequenceName:Text=GetJobName!(InMoviePipeline:Movie Pipeline Object Reference=@n1) -> (ReturnValue:Text)
    ) -> (Result)
   )
   n2 = GetJobInitializationTime!(InMoviePipeline:Movie Pipeline Object Reference=@n1) -> (ReturnValue:Date Time Structure)
   > "SetText (Text)"(
    self=Get Top_02() -> (Top_02:Text Object Reference)
    InText=Format Text!(
     Format="LOCTEXT(\"{Date}\")"
     Date:Text="As Date (from UTC)"!(InDateTime:"Date Time Structure (by ref)"=@n2, InTimeZone:String, InDateStyle:EDateTimeStyle Enum=Default) -> (ReturnValue:Text)
    ) -> (Result)
   )
   > "SetText (Text)"(
    self=Get Top_04() -> (Top_04:Text Object Reference)
    InText=Format Text!(Format="LOCTEXT(\"{AuthorName}\")", AuthorName:Text=GetJobAuthor!(InMoviePipeline:Movie Pipeline Object Reference=@n1) -> (ReturnValue:Text)) -> (Result)
   )
  ]}
 UpdateTopRightText {nodes [
   n1 = UpdateTopRightText!() -> (then, ForPipeline:Movie Pipeline Object Reference) {is=title K2Node_FunctionEntry}
   > n2 = Sequence!(execute) -> (then_0, then_1, then_2, then_3)
   > Set ChangeList(ChangeList:String="To String (Text)"(InText=GetMoviePipelineEngineChangelistLabel!(InMoviePipeline:Movie Pipeline Object Reference=@n1) -> (ReturnValue:Text))) -> (Output_Get:String)
   n3 = Get BranchName() -> (BranchName:String)
   n4 = Get ChangeList() -> (ChangeList:String)
   "SetText (Text)"(
    execute=@n2.then_2
    self=Get TopRight() -> (TopRight:Text Object Reference)
    InText=Format Text!(
     Format="LOCTEXT(\"{BranchName}-{ChangeList}\")"
     BranchName:Text="To Text (String)"(InString=@n3)
     ChangeList:Text="To Text (String)"(InString=@n4)
    ) -> (Result)
   )
   n5 = GetEngineVersion!() -> (ReturnValue:String)
   n6 = FindSubstring(SearchIn=@n5, Substring=+, bSearchFromEnd=true)
   Branch(execute=@n2.then_1, Condition="integer >= integer"!(A:int=@n6, B:int=0) -> (ReturnValue:bool) {is=title K2Node_CallFunction})
   n7 = int + int(A=@n6, B=1) -> (ReturnValue=0) {is=title K2Node_CommutativeAssociativeBinaryOperator}
   > Set BranchName(BranchName:String=GetSubstring!(SourceString:String=@n5, StartIndex:int=@n7, Length:int=int - int(A=@n6, B=1) {is=title K2Node_CallFunction}) -> (ReturnValue:String)) -> (Output_Get:String)
   int - int(A=Len(S=@n5), B=@n7)
   FindOrGetDefaultSettingForShot!(
    execute
    InSettingType:Movie Pipeline Setting Class Reference=/Script/MovieRenderPipelineCore.MoviePipelineAntiAliasingSetting
    InPrimaryConfig:Movie Pipeline Primary Config Object Reference=GetPipelinePrimaryConfig!(self:Movie Pipeline Object Reference=@n1) -> (ReturnValue:Movie Pipeline Primary Config Object Reference)
    InShot:Movie Pipeline Executor Shot Object Reference=GetCurrentExecutorShot!(InMoviePipeline:Movie Pipeline Object Reference=@n1) -> (ReturnValue:Movie Pipeline Executor Shot Object Reference)
   ) -> (then, ReturnValue:Movie Pipeline Anti Aliasing Setting Object Reference)
  ]}
 UpdateUserCommentText {nodes [
   n1 = GetCurrentJob!(self:Movie Pipeline Object Reference=UpdateUserCommentText!() -> (then, For Pipeline:Movie Pipeline Object Reference) {is=title K2Node_FunctionEntry}) -> (ReturnValue:Movie Pipeline Executor Job Object Reference)
   Set Text(
    Text="To Text (String)"(InString=Get Comment!(self:Movie Pipeline Executor Job Object Reference=@n1) -> (Comment:String))
    self:"Text Box (Multi- Line) Object Reference"
   )
  ]}
]
variables []
widgets {
 info {
  inheritedWidgetCount=0
  parentClass=/Script/MovieRenderPipelineSettings.MoviePipelineBurnInWidget
  rootWidgetClass=&Overlay
  widgetCount=23
 }
 widgets [
  Overlay_45 {widgetClassPath=&Overlay, widget=Overlay_45}
  RenderedImagePlaceholder {
   widgetClassPath=&Image
   parent=Overlay_45
   widget=RenderedImagePlaceholder
   properties {
    visibility=Hidden
    brush {
     bIsDynamicallyLoaded=false
     drawAs=Image
     imageType=NoImage
     mirroring=NoMirror
     resourceName=None
     resourceObject=null
     tiling=NoTile
     imageSize=<1920 1080>
     margin=<ltrb 0 0 0 0>
     outlineSettings {
      bUseBrushTransparency=false
      roundingType=HalfHeightRadius
      width=0
      color {colorUseRule=UseColor_Specified, specifiedColor=<rgba 0 0 0 0>}
      cornerRadii=<0 0 0 0>
     }
     tintColor {colorUseRule=UseColor_Specified, specifiedColor=<rgba 0.026 0.026 0.026 1>}
     uVRegion {bIsValid=false, max=<0 0>, min=<0 0>}
    }
   }
   slot {class=OverlaySlot, properties {horizontalAlignment=HAlign_Fill, verticalAlignment=VAlign_Fill}}
  }
  VerticalBox_41 {
   widgetClassPath=&VerticalBox
   parent=Overlay_45
   widget=VerticalBox_41
   slot {class=OverlaySlot, properties {horizontalAlignment=HAlign_Fill, verticalAlignment=VAlign_Fill}}
  }
  HorizontalBox_22 {
   widgetClassPath=&HorizontalBox
   parent=VerticalBox_41
   widget=HorizontalBox_22
   slot {class=VerticalBoxSlot, properties {padding=<ltrb 0 10 0 4>}}
  }
  Spacer_1 {
   widgetClassPath=&Spacer
   parent=HorizontalBox_22
   widget=Spacer_1
   slot {class=HorizontalBoxSlot, properties {padding=<ltrb 20 20 20 20>}}
  }
  VerticalBox {widgetClassPath=&VerticalBox, parent=HorizontalBox_22, widget=VerticalBox, slot {class=HorizontalBoxSlot}}
  Top_04 {
   widgetClassPath=&TextBlock
   bIsVariable=true
   parent=VerticalBox
   widget=Top_04
   properties {
    text=Author
    font {
     bForceMonospaced=false
     bMaterialIsStencil=false
     fontMaterial=null
     fontObject=&NotoSans
     letterSpacing=0
     monospacedWidth=1
     size=36
     skewAmount=0
     typefaceFontName=NotoSans-Regular
     outlineSettings {
      bApplyOutlineToDropShadows=false
      bMiteredCorners=false
      bSeparateFillAlpha=false
      outlineMaterial=null
      outlineSize=0
      outlineColor=<rgba 0 0 0 1>
     }
    }
   }
   slot {class=VerticalBoxSlot, properties {horizontalAlignment=HAlign_Center, verticalAlignment=VAlign_Center}}
  }
  Spacer_2 {
   widgetClassPath=&Spacer
   parent=HorizontalBox_22
   widget=Spacer_2
   slot {class=HorizontalBoxSlot, properties {size {sizeRule=Fill, value=1}}}
  }
  VerticalBox_144 {widgetClassPath=&VerticalBox, parent=HorizontalBox_22, widget=VerticalBox_144, slot {class=HorizontalBoxSlot}}
  HorizontalBox_118 {
   widgetClassPath=&HorizontalBox
   parent=VerticalBox_144
   widget=HorizontalBox_118
   slot {class=VerticalBoxSlot, properties {horizontalAlignment=HAlign_Right}}
  }
  Top_02 {
   widgetClassPath=&TextBlock
   bIsVariable=true
   parent=HorizontalBox_118
   widget=Top_02
   properties {
    text=01/19/2020
    font {
     bForceMonospaced=false
     bMaterialIsStencil=false
     fontMaterial=null
     fontObject=&NotoSans
     letterSpacing=0
     monospacedWidth=1
     size=27
     skewAmount=0
     typefaceFontName=NotoSans-Regular
     outlineSettings {
      bApplyOutlineToDropShadows=false
      bMiteredCorners=false
      bSeparateFillAlpha=false
      outlineMaterial=null
      outlineSize=0
      outlineColor=<rgba 0 0 0 1>
     }
    }
   }
   slot {class=HorizontalBoxSlot, properties {horizontalAlignment=HAlign_Left, verticalAlignment=VAlign_Top}}
  }
  HorizontalBox_208 {
   widgetClassPath=&HorizontalBox
   parent=VerticalBox_144
   widget=HorizontalBox_208
   slot {class=VerticalBoxSlot, properties {horizontalAlignment=HAlign_Right}}
  }
  TopRight {
   widgetClassPath=&TextBlock
   bIsVariable=true
   parent=HorizontalBox_208
   widget=TopRight
   properties {
    justification=Right
    text=Dev-Editor-11047590
    font {
     bForceMonospaced=false
     bMaterialIsStencil=false
     fontMaterial=null
     fontObject=&NotoSans
     letterSpacing=0
     monospacedWidth=1
     size=18
     skewAmount=0
     typefaceFontName=NotoSans-Regular
     outlineSettings {
      bApplyOutlineToDropShadows=false
      bMiteredCorners=false
      bSeparateFillAlpha=false
      outlineMaterial=null
      outlineSize=0
      outlineColor=<rgba 0 0 0 1>
     }
    }
   }
   slot {class=HorizontalBoxSlot, properties {horizontalAlignment=HAlign_Right}}
  }
  HorizontalBox_302 {
   widgetClassPath=&HorizontalBox
   parent=VerticalBox_144
   widget=HorizontalBox_302
   slot {class=VerticalBoxSlot, properties {horizontalAlignment=HAlign_Right}}
  }
  Top_01 {
   widgetClassPath=&TextBlock
   bIsVariable=true
   parent=HorizontalBox_302
   widget=Top_01
   properties {
    text=Character
    font {
     bForceMonospaced=false
     bMaterialIsStencil=false
     fontMaterial=null
     fontObject=&NotoSans
     letterSpacing=0
     monospacedWidth=1
     size=27
     skewAmount=0
     typefaceFontName=NotoSans-Regular
     outlineSettings {
      bApplyOutlineToDropShadows=false
      bMiteredCorners=false
      bSeparateFillAlpha=false
      outlineMaterial=null
      outlineSize=0
      outlineColor=<rgba 0 0 0 1>
     }
    }
   }
   slot {class=HorizontalBoxSlot, properties {horizontalAlignment=HAlign_Right}}
  }
  Spacer {
   widgetClassPath=&Spacer
   parent=HorizontalBox_22
   widget=Spacer
   slot {class=HorizontalBoxSlot, properties {padding=<ltrb 15 15 15 15>}}
  }
  Spacer_60 {
   widgetClassPath=&Spacer
   parent=VerticalBox_41
   widget=Spacer_60
   slot {class=VerticalBoxSlot, properties {size {sizeRule=Fill, value=1}}}
  }
  HorizontalBox {
   widgetClassPath=&HorizontalBox
   parent=VerticalBox_41
   widget=HorizontalBox
   slot {class=VerticalBoxSlot, properties {padding=<ltrb 4 4 4 4>}}
  }
  Spacer_256 {
   widgetClassPath=&Spacer
   parent=HorizontalBox
   widget=Spacer_256
   slot {class=HorizontalBoxSlot, properties {size {sizeRule=Fill, value=1}}}
  }
  BottomLeft {
   widgetClassPath=&TextBlock
   bIsVariable=true
   parent=HorizontalBox
   widget=BottomLeft
   properties {
    text=344f
    font {
     bForceMonospaced=false
     bMaterialIsStencil=false
     fontMaterial=null
     fontObject=&NotoSans
     letterSpacing=0
     monospacedWidth=1
     size=27
     skewAmount=0
     typefaceFontName=NotoSans-Regular
     outlineSettings {
      bApplyOutlineToDropShadows=false
      bMiteredCorners=false
      bSeparateFillAlpha=false
      outlineMaterial=null
      outlineSize=0
      outlineColor=<rgba 0 0 0 1>
     }
    }
   }
   slot {class=HorizontalBoxSlot, properties {horizontalAlignment=HAlign_Left, size {sizeRule=Automatic, value=0.5}}}
  }
  Spacer_3 {
   widgetClassPath=&Spacer
   parent=HorizontalBox
   widget=Spacer_3
   slot {class=HorizontalBoxSlot, properties {padding=<ltrb 20 20 20 20>}}
  }
  HorizontalBox_0 {
   widgetClassPath=&HorizontalBox
   parent=HorizontalBox
   widget=HorizontalBox_0
   slot {class=HorizontalBoxSlot, properties {horizontalAlignment=HAlign_Right, size {sizeRule=Automatic, value=0.5}}}
  }
  Spacer_4 {
   widgetClassPath=&Spacer
   parent=VerticalBox_41
   widget=Spacer_4
   slot {class=VerticalBoxSlot, properties {padding=<ltrb 10 10 10 10>}}
  }
 ]
}
