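# Editor utility widget for rendering facial ROM movies of a character head or companion CID in the FaceROM level and uploading them to ShotGrid.
# Asset pickers set the CID path, preview icon and frame range, or apply companion definitions and ROM montages to mannequins.
# Render rebuilds the Movie Render Queue job, runs it in PIE, then turns frames into a video on success.
# Most pipeline work runs through inline Python calling TechAnim.faceDeltaRender.
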EditorUtilityWidgetBlueprint /Game/NeverCook/ArtTools/Tech_Anim/FaceRom_Render_EUW
graphs [
 EUW_Create_Movie_File {nodes [
   EUW_Create_Movie_File() {is=title K2Node_FunctionEntry}
   > n1 = FindAssetData!(execute, AssetPath:String=Get CID_Path() -> (CID_Path:String)) -> (then, ReturnValue:Asset Data Structure)
   > Execute Python Script!(
    execute
    PythonScript="import unreal\r\nimport importlib\r\nimport TechAnim.faceDeltaRender as fdr\r\nimportlib.reload(fdr)\r\n\r\nbase_name = cid.get_name()\r\nfdr.generate_video_from_frames(base_name, frame_source_folder=output_path, remove_source_images=True)"
    cid:Object Reference=GetAsset(InAssetData=@n1)
    output_path:String=Break Directory Path!(DirectoryPath:"Directory Path Structure (by ref)"=Get OutputDirectory() -> (OutputDirectory:Directory Path Structure)) -> (Path:String) {is=title K2Node_BreakStruct}
   ) -> (then, ReturnValue) {inputs [cid,output_path]}
   > n2 = EUW_Get_Output_Filepath!(execute, self) -> (then, filepath:String)
   > SetIsEnabled(
    self=Get Button_Upload_ROM() -> (Button_Upload_ROM:Editor Utility Button Object Reference)
    bInIsEnabled=FileExists!(InPath:String=@n2) -> (ReturnValue:bool)
   )
   > EUW_ToggleUIDisplay!(execute, self, Visible:bool=true) -> (then)
   > Show Message Dialog(
    Title=Get Info Title String() -> (Info Title String:Text)
    Message=Get Render Done Message String() -> (Render Done Message String:Text)
    MessageType=Ok
    DefaultValue=Ok
    MessageCategory=Error
   )
  ]}
 EUW_Get_Output_Filepath {nodes [
   EUW_Get_Output_Filepath() {is=title K2Node_FunctionEntry}
   > n1 = FindAssetData(AssetPath=Get CID_Path())
   > n2 = Execute Python Script!(
    execute
    PythonScript="import unreal\r\nimport importlib\r\nimport TechAnim.faceDeltaRender as fdr\r\nimportlib.reload(fdr)\r\n\r\noutput_filename = fdr.get_render_movie_file_path(cid, output_dir)"
    cid=GetAsset(InAssetData=@n1)
    output_dir:String=Break Directory Path(DirectoryPath=Get OutputDirectory())
   ) -> (then, ReturnValue, output_filename:String) {inputs [cid,output_dir], outputs [output_filename]}
   > Return Node!(execute, filepath:String=@n2.output_filename)
  ]}
 EUW_Reset_CIDs {nodes [
   n1 = EUW_Reset_CIDs!() -> (then, Reset Head:bool=~, Reset Companion:bool=~) {is=title K2Node_FunctionEntry}
   > Set CIDAssetPreviewIconBrush(CIDAssetPreviewIconBrush:Slate Brush Structure=Get CIDAssetPreviewDefaultIconBrush() -> (CIDAssetPreviewDefaultIconBrush:Slate Brush Structure)) -> (Output_Get:Slate Brush Structure)
   > n2 = Sequence()
   > Branch(Condition=@n1)
   > Set Head_CID(Head_CID:Athena Character Item Definition Object Reference) -> (Output_Get:Athena Character Item Definition Object Reference)
   Branch(execute=@n2.then_1, Condition=@n1.Reset Companion)
   > Set Companion Target CID(Companion Target CID:Cosmetic Companion Item Definition Object Reference) -> (Output_Get:Cosmetic Companion Item Definition Object Reference)
  ]}
 EUW_Set_Preview_Icon_and_CID_Path {nodes [
   EUW_Set_Preview_Icon_and_CID_Path() {is=title K2Node_FunctionEntry}
   > n1 = Sequence()
   > n2 = Branch(Condition=IsChecked(self=Get Companion_checkBox() -> (Companion_checkBox:Editor Utility Check Box Object Reference)))
   > Set CIDAssetPreviewIconBrush(CIDAssetPreviewIconBrush=GetItemDefinitionSmallPreviewImageBrush!(
     ItemDefinition:Fort Item Definition Object Reference=Get Target() -> (Target:Cosmetic Companion Item Definition Object Reference)
     BrushSize:EFortBrushSize Enum=Large
    ) -> (ReturnValue:Slate Brush Structure))
   > Set CID_Path(CID_Path:String=Get Object Path String(Object=Get Target())) -> (Output_Get:String)
   Set CIDAssetPreviewIconBrush(
    execute=@n2.else
    CIDAssetPreviewIconBrush=GetItemDefinitionSmallPreviewImageBrush(ItemDefinition=Get Head_CID() -> (Head_CID:Athena Character Item Definition Object Reference), BrushSize=Large)
   )
   > Set CID_Path(CID_Path=Get Object Path String(Object=Get Head_CID()))
   Return Node!(execute=@n1.then_1, CID_Path:String=Get CID_Path())
  ]}
 EUW_ToggleUIDisplay {nodes [
   n1 = EUW_ToggleUIDisplay!() -> (then, Visible:bool=~) {is=title K2Node_FunctionEntry}
   > n2 = GetEditorWorld(self=Get UnrealEditorSubsystem())
   > n3 = GetActorOfClass!(execute, WorldContextObject:Object Reference=@n2, ActorClass:Actor Class Reference=/Game/ArtTools/ROMs/BP_BodyROM.BP_BodyROM_C) -> (then, ReturnValue:BP Body ROM Object Reference)
   > Is Valid(InputObject=@n3)
   > Set displayUI!(execute, displayUI:bool=@n1, self:BP Body ROM Object Reference=@n3) -> (then, Output_Get:bool)
  ]}
 EUW_Update_By_CID {nodes [
   EUW_Update_By_CID() {is=title K2Node_FunctionEntry}
   > n1 = EUW_Set_Preview_Icon_and_CID_Path!(execute, self) -> (then, CID_Path:String) {is=title K2Node_CallFunction}
   > PrintString(InString=Append(A=CID PATH, B=@n1))
   > n2 = FindAssetData(AssetPath=@n1)
   n3 = GetAsset(InAssetData=@n2)
   > PrintString(InString=Append(A=GET ASSET, B=GetDisplayName(Object=@n3)))
   > n4 = Is Valid(InputObject=@n3)
   PrintString(execute=@n4.Is Not Valid, InString=CIDDDDDDDDDDDDDDDDD:::::::::)
   > Show Message Dialog(
    Title=Get Error Title String() -> (Error Title String:Text)
    Message=Get Invalid CID Message String() -> (Invalid CID Message String:Text)
    MessageType=Ok
    DefaultValue=Ok
    MessageCategory=Error
   )
   n5 = Execute Python Script!(
    execute=@n4.Is Valid
    PythonScript="import unreal\r\nimport importlib\r\nimport TechAnim.faceDeltaRender as fdr\r\nimportlib.reload(fdr)\r\n\r\nhead_type = ''\r\ngender = ''\r\nstart_frame = 0\r\nend_frame = 1\r\n\r\ninfo = fdr.get_head_info_from_cid(cid)\r\nhead_type = info.get('head_type', '')\r\ngender = info.get('gender', '')\r\n\r\nstart_frame, end_frame = fdr.get_framerange_from_cid(cid)"
    cid=@n3
   ) -> (then, ReturnValue, head_type:Text, gender:Text, start_frame:float=~, end_frame:float=~) {inputs [cid], outputs [head_type,gender,start_frame,end_frame]}
   > Set headTypeText(headTypeText:Text=@n5.head_type) -> (Output_Get:Text)
   > Set gender(gender:Text=@n5.gender) -> (Output_Get:Text)
   n6 = Make Float Range!(
    LowerBound:Float Range Bound Structure=Make Float Range Bound!(Type:ERangeBoundTypes Enum=Inclusive, Value:float=@n5.start_frame) -> (FloatRangeBound:Float Range Bound Structure) {is=title K2Node_MakeStruct}
    UpperBound:Float Range Bound Structure=Make Float Range Bound(Type=Inclusive, Value=@n5.end_frame)
   ) -> (FloatRange:Float Range Structure) {is=title K2Node_MakeStruct}
   > Set FrameRange(FrameRange:Float Range Structure=@n6) -> (Output_Get:Float Range Structure)
   > Set Default_FrameRange(Default_FrameRange:Float Range Structure=@n6) -> (Output_Get:Float Range Structure)
   > Execute Python Script!(
    execute
    PythonScript="import unreal\r\nimport importlib\r\nimport TechAnim.faceDeltaRender as fdr\r\nimportlib.reload(fdr)  \r\n\r\nfdr.set_CID_main(cid)"
    cid=@n3
   ) -> (then, ReturnValue) {inputs [cid]}
   > n7 = EUW_Get_Output_Filepath() {is=title K2Node_CallFunction}
   > SetIsEnabled(self=Get Button_Upload_ROM(), bInIsEnabled=FileExists(InPath=@n7))
   Return Node!(execute)
  ]}
 EUW_Validate_Level {nodes [
   EUW_Validate_Level() {is=title K2Node_FunctionEntry}
   > n1 = GetCurrentLevelName!(execute, WorldContextObject:Object Reference, bRemovePrefixString:bool=true) -> (then, ReturnValue:String)
   > n2 = Branch(Condition="Equal Exactly (String)"(A=@n1))
   Show Message Dialog(
    execute=[@n2.then, @n3.else]
    Title=Get Error Title String()
    Message=Get Level Error Message String() -> (Level Error Message String:Text)
    MessageType=Ok
    DefaultValue=Ok
    MessageCategory=Error
   )
   > Return Node!(execute, level_is_valid:bool=false)
   n3 = Branch(execute=@n2.else, Condition="Equal Exactly (String)"(A=@n1, B=FaceROM))
   > Return Node(level_is_valid=true)
  ]}
 EventGraph {nodes [
   n1 = "On Check State Changed (Companion_checkBox)"!() -> (OutputDelegate, then, bIsChecked:bool) {is=title K2Node_ComponentBoundEvent}
   > EUW_Reset_CIDs!(execute, self, Reset Head:bool=true, Reset Companion:bool=true) -> (then) {is=title K2Node_CallFunction}
   > n2 = Branch(Condition=@n1.bIsChecked)
   n3 = Get AssetsDetailsView() -> (AssetsDetailsView:Details View Object Reference)
   > SetVisibility(self=@n3, InVisibility=Collapsed)
   n4 = Get CompanionTargetCID() -> (CompanionTargetCID:Details View Object Reference)
   > SetVisibility(self=@n4, InVisibility=Visible)
   n5 = Get CompanionSourceCID() -> (CompanionSourceCID:Details View Object Reference)
   > SetVisibility(self=@n5, InVisibility=Visible)
   n6 = Get CompanionFaceROM() -> (CompanionFaceROM:Details View Object Reference)
   > SetVisibility(self=@n6, InVisibility=Visible)
   SetVisibility(execute=@n2.else, self=@n3, InVisibility=Visible)
   > SetVisibility(self=@n4, InVisibility=Collapsed)
   > SetVisibility(self=@n5, InVisibility=Collapsed)
   > SetVisibility(self=@n6, InVisibility=Collapsed)
   "On Clicked (Button_Reset_Render_option)"() {is=title K2Node_ComponentBoundEvent}
   > Set FrameRange(FrameRange=Get Default_FrameRange() -> (Default_FrameRange:Float Range Structure))
   > Set OutputDirectory(OutputDirectory:Directory Path Structure=Make Directory Path!(Path:String=D:/FaceROMs) -> (DirectoryPath:Directory Path Structure) {is=title K2Node_MakeStruct}) -> (Output_Get:Directory Path Structure)
   "On Property Changed (CompanionFaceROM)"!() -> (OutputDelegate, then, PropertyName:Name=None) {is=title K2Node_ComponentBoundEvent}
   > n7 = Cast To AnimMontage!(execute, Object=Get ROM() -> (ROM:Anim Montage Object Reference)) -> (then, CastFailed, AsAnim Montage:Anim Montage Object Reference) {is=title K2Node_DynamicCast}
   > Set Companion Anim Montage(Companion Anim Montage:Anim Montage Object Reference=@n7) -> (Output_Get:Anim Montage Object Reference)
   > n8 = For Each Loop(Array:Array of Fort Companion Mannequin Object References=Get Fort Companion Mannequins() -> (Fort Companion Mannequins:Array of Fort Companion Mannequin Object References)) -> (Array Element:Fort Companion Mannequin Object Reference)
   > n9 = Set CompanionSM Component(CompanionSM Component:Skeletal Mesh Component Object Reference=GetComponentByClass!(self:Actor Object Reference=@n8, ComponentClass:Actor Component Class Reference=/Script/Engine.SkeletalMeshComponent) -> (ReturnValue:Skeletal Mesh Component Object Reference)) -> (Output_Get:Skeletal Mesh Component Object Reference)
   > SetAnimationMode!(
    execute
    self:Skeletal Mesh Component Object Reference=@n9
    InAnimationMode:EAnimationMode Enum=AnimationSingleNode
    bForceInitAnimScriptInstance:bool=false
   ) -> (then)
   > Delay!(execute, WorldContextObject:Object Reference, Duration=2) -> (then)
   > n10 = GetPathNameForLoadedAsset(LoadedAsset=Get Companion Anim Montage() -> (Companion Anim Montage:Anim Montage Object Reference))
   > Execute Python Script!(
    execute
    PythonScript="import unreal\r\nimport importlib\r\nimport TechAnim.faceDeltaRender as fdr\r\nimportlib.reload(fdr)\r\nfdr.set_companion_rom_anim(rom)"
    rom:String=@n10
   ) -> (then, ReturnValue) {inputs [rom]}
   "On Clicked (Button_Open_Level)"() {is=title K2Node_ComponentBoundEvent}
   > Execute Python Script!(
    execute
    PythonScript="import importlib\r\nimport unreal\r\nimport TechAnim.faceDeltaRender as fdr\r\nimportlib.reload(fdr)\r\nfdr.open_FaceRom_level()"
   ) -> (then, ReturnValue)
   "On Property Changed (CompanionSourceCID)"!() -> (OutputDelegate, then, PropertyName:Name=None) {is=title K2Node_ComponentBoundEvent}
   > Set ROM(ROM:Anim Montage Object Reference) -> (Output_Get:Anim Montage Object Reference)
   > n11 = Sequence()
   > n12 = EUW_Validate_Level!(execute, self) -> (then, level_is_valid:bool) {is=title K2Node_CallFunction}
   > n13 = Branch(Condition=@n12)
   > EUW_Update_By_CID() {is=title K2Node_CallFunction}
   > Delay(Duration=2)
   > Execute Python Script(PythonScript="import unreal\r\nimport importlib\r\nimport TechAnim.faceDeltaRender as fdr\r\nimportlib.reload(fdr)\r\nfdr.adjust_companion_mann_position()")
   EUW_Reset_CIDs(execute=@n13.else, Reset Head=true, Reset Companion=true)
   n14 = Cast To CosmeticCompanionItemDefinition!(execute=@n11.then_1, Object=Get Source() -> (Source:Cosmetic Companion Item Definition Object Reference)) -> (then, CastFailed, AsCosmetic Companion Item Definition:Cosmetic Companion Item Definition Object Reference) {is=title K2Node_DynamicCast}
   > Set Companion Source CID(Companion Source CID:Cosmetic Companion Item Definition Object Reference=@n14) -> (Output_Get:Cosmetic Companion Item Definition Object Reference)
   > n15 = GetAllActorsOfClass!(execute, WorldContextObject:Object Reference, ActorClass:Actor Class Reference=/Script/CosmeticCompanionsRuntime.FortCompanionMannequin) -> (then, OutActors:Array of Fort Companion Mannequin Object References)
   > n16 = For Each Loop(Array=@n15)
   > Branch(Condition=Contains(SearchIn=GetDisplayName(Object=@n16), Substring=_Ref, bUseCase=true))
   > SetEditorProperty(
    Object=@n16
    PropertyName=CompanionItemDefinition
    PropertyValue:Cosmetic Companion Item Definition Object Reference=Get Companion Source CID() -> (Companion Source CID:Cosmetic Companion Item Definition Object Reference)
   )
   > Add(
    TargetArray:Array of Fort Companion Mannequin Object References=Get Fort Companion Mannequins()
    NewItem:"Fort Companion Mannequin Object Reference (by ref)"=@n16
   ) {is=title K2Node_CallArrayFunction}
   "On Clicked (Button_Render_ROM)"() {is=title K2Node_ComponentBoundEvent}
   > n17 = FindAssetData(AssetPath=Get CID_Path())
   n18 = GetAsset(InAssetData=@n17)
   > n19 = Is Valid(InputObject=@n18)
   Show Message Dialog(
    execute=@n19.Is Not Valid
    Title=Get Error Title String()
    Message=Get Invalid CID Message String()
    MessageType=Ok
    DefaultValue=Ok
    MessageCategory=Error
   )
   n20 = EUW_Validate_Level(execute=@n19.Is Valid)
   > Branch(Condition=@n20)
   > n21 = Sequence!(execute) -> (then_0, then_1, then_2:Exec)
   n22 = Get MoviePipelineQueueSubsystem() -> (ReturnValue:Movie Pipeline Queue Subsystem Object Reference) {is=title K2Node_GetEditorSubsystem}
   > DeleteAllJobs(self:Movie Pipeline Queue Object Reference=GetQueue!(self:Movie Pipeline Queue Subsystem Object Reference=@n22) -> (ReturnValue:Movie Pipeline Queue Object Reference))
   n23 = Break Capture Resolution!(CaptureResolution:"Capture Resolution Structure (by ref)"=Get resolution() -> (resolution:Capture Resolution Structure)) -> (ResX:int, ResY:int) {is=title K2Node_BreakStruct}
   n24 = Break Float Range!(FloatRange:"Float Range Structure (by ref)"=Get FrameRange() -> (FrameRange:Float Range Structure)) -> (LowerBound:Float Range Bound Structure, UpperBound:Float Range Bound Structure) {is=title K2Node_BreakStruct}
   > Execute Python Script!(
    execute
    PythonScript="import unreal\r\nimport importlib\r\nimport TechAnim.faceDeltaRender as fdr\r\nimportlib.reload(fdr)\r\n\r\ntry:    \r\n    fdr.create_render_job(cid, output_path,  render_settings, start_frame=start_frame, end_frame=end_frame, x_res=x_res, y_res=y_res, level=level, level_sequence=level_sequence)\r\nexcept Exception as error:\r\n    unreal.log_error(\"Failed to Render\")    \r\n    unreal.log_error(error)\r\n"
    cid:Object Reference=@n18
    output_path:String=Break Directory Path(DirectoryPath=Get OutputDirectory())
    render_settings:Movie Pipeline Primary Config Object Reference=Get MRQSettings() -> (MRQSettings:Movie Pipeline Primary Config Object Reference)
    start_frame:float=Break Float Range Bound!(FloatRangeBound:"Float Range Bound Structure (by ref)"=@n24) -> (Type:ERangeBoundTypes Enum, Value:float=~) {is=title K2Node_BreakStruct}.Value
    end_frame:float=Break Float Range Bound(FloatRangeBound=@n24.UpperBound) -> (Value=~).Value
    x_res:int=@n23
    y_res:int=@n23.ResY
    level:World Soft Object Reference=Break Soft World Reference!(SoftWorldReference:"Soft World Reference Structure (by ref)"=Get WorkingLevel() -> (WorkingLevel:Soft World Reference Structure)) -> (WorldAsset:World Soft Object Reference) {is=title K2Node_BreakStruct}
    level_sequence:Level Sequence Soft Object Reference=Get LevelSequence() -> (LevelSequence:Level Sequence Soft Object Reference)
   ) -> (then, ReturnValue) {inputs [cid, output_path, render_settings, start_frame, end_frame, x_res, y_res, level, level_sequence]}
   EUW_ToggleUIDisplay(execute=@n21.then_1, Visible=false) {is=title K2Node_CallFunction}
   n25 = RenderQueueWithExecutor!(
    execute=@n21.then_2
    self:Movie Pipeline Queue Subsystem Object Reference=@n22
    InExecutorType:Movie Pipeline Executor Base Class Reference=/Script/MovieRenderPipelineEditor.MoviePipelinePIEExecutor
   ) -> (then, ReturnValue:Movie Pipeline PIEExecutor Object Reference)
   n26 = RenderFinished!() -> (OutputDelegate, then, PipelineExecutor:Movie Pipeline Executor Base Object Reference, bSuccess:bool=~) {is=title K2Node_CustomEvent}
   Bind Event to On Executor Finished Delegate(execute=@n25.then, self:Movie Pipeline Executor Base Object Reference=@n25, Delegate=@n26) {is=title K2Node_AddDelegate}
   "On Property Changed (AssetsDetailsView)"!() -> (OutputDelegate, then, PropertyName:Name=None) {is=title K2Node_ComponentBoundEvent}
   > n27 = EUW_Validate_Level()
   > n28 = Branch(Condition=@n27)
   > EUW_Update_By_CID()
   EUW_Reset_CIDs(execute=@n28.else, Reset Head=true, Reset Companion=true)
   "On Property Changed (CompanionTargetCID)"!() -> (OutputDelegate, then, PropertyName:Name=None) {is=title K2Node_ComponentBoundEvent}
   > Set ROM()
   > Set Source(Source:Cosmetic Companion Item Definition Object Reference) -> (Output_Get:Cosmetic Companion Item Definition Object Reference)
   > n29 = EUW_Set_Preview_Icon_and_CID_Path()
   > Set CID_Path(CID_Path=@n29)
   > n30 = Cast To CosmeticCompanionItemDefinition(Object=Get Target())
   > Set Companion Target CID(Companion Target CID=@n30)
   > n31 = GetAllActorsOfClass(ActorClass=&FortCompanionMannequin)
   > n32 = For Each Loop(Array=@n31)
   n33 = GetDisplayName(Object=@n32)
   > Branch(Condition=OR Boolean(A="Equal Exactly (String)"(A=@n33, B=FortCompanionMannequin_Front), B="Equal Exactly (String)"(A=@n33, B=FortCompanionMannequin_Side)))
   > SetEditorProperty(
    Object=@n32
    PropertyName=CompanionItemDefinition
    PropertyValue=Get Companion Target CID() -> (Companion Target CID:Cosmetic Companion Item Definition Object Reference)
   )
   > Add(TargetArray=Get Fort Companion Mannequins(), NewItem=@n32)
   "On Clicked (OpenInSGButton)"() {is=title K2Node_ComponentBoundEvent}
   > n34 = FindAssetData(AssetPath=Get CID_Path())
   n35 = GetAsset(InAssetData=@n34)
   > n36 = Is Valid(InputObject=@n35)
   Show Message Dialog(
    execute=@n36.Is Not Valid
    Title=Get Error Title String()
    Message=Get Invalid CID Message String()
    MessageType=Ok
    DefaultValue=Ok
    MessageCategory=Error
   )
   Execute Python Script!(
    execute=@n36.Is Valid
    PythonScript="import unreal\r\nimport importlib\r\nimport TechAnim.faceDeltaRender as fdr\r\nimportlib.reload(fdr)\r\n\r\nhead_sg_id = fdr.get_head_sg_id(cid)\r\nif head_sg_id:\r\n    import webbrowser\r\n    asset_url = f'https://epicgames.shotgunstudio.com/detail/Asset/{head_sg_id}'\r\n    webbrowser.open(asset_url)\r\nelse:\r\n    raise RuntimeError(f'cannot find sg id:{head_sg_id} for {cid}!!!')"
    cid=@n35
   ) -> (then, ReturnValue) {inputs [cid]}
   "On Clicked (Button_Upload_ROM)"() {is=title K2Node_ComponentBoundEvent}
   > n37 = FindAssetData(AssetPath=Get CID_Path())
   n38 = GetAsset(InAssetData=@n37)
   > n39 = Is Valid(InputObject=@n38)
   Show Message Dialog(
    execute=@n39.Is Not Valid
    Title=Get Error Title String()
    Message=Get Invalid CID Message String()
    MessageType=Ok
    DefaultValue=Ok
    MessageCategory=Error
   )
   Execute Python Script!(
    execute=@n39.Is Valid
    PythonScript="import unreal\r\nimport importlib\r\nimport TechAnim.faceDeltaRender as fdr\r\nimportlib.reload(fdr)\r\n\r\nfdr.upload_render_to_shotgrid(cid, output_dir)"
    cid=@n38
    output_dir:String=Break Directory Path(DirectoryPath=Get OutputDirectory())
   ) -> (then, ReturnValue) {inputs [cid,output_dir]}
   Event PreConstruct()
   > SetObject(self=Get AssetsDetailsView(), NewObject=Self-Reference())
   > SetObject(self=Get Render_DetailsView() -> (Render_DetailsView:Details View Object Reference), NewObject=Self-Reference())
   > SetObject(self=Get CompanionTargetCID(), NewObject=Self-Reference())
   > SetObject(self=Get CompanionSourceCID(), NewObject=Self-Reference())
   > SetObject(self=Get CompanionFaceROM(), NewObject=Self-Reference())
   > Execute Python Script!(
    execute
    PythonScript="\r\ntry:\r\n   from unpipe import get_active_integration\r\n   get_active_integration()\r\nexcept ImportError:\r\n   try:\r\n       from pipeline_menus import PipelineMenu\r\n       PipelineMenu.load_pipeline_tools()\r\n   except Exeception as e:\r\n       raise RuntimeError(\"Failed to bootstrap the Pipeline\") from e\r\n"
   ) -> (then, ReturnValue)
   EUW_Create_Movie_File(execute=@n40.then) {is=title K2Node_CallFunction}
   Set Resolution(Resolution:Capture Resolution Structure=Make Capture Resolution!(ResX:int=1920, ResY:int=1080) -> (CaptureResolution:Capture Resolution Structure) {is=title K2Node_MakeStruct}) -> (Output_Get:Capture Resolution Structure)
   n40 = Branch(execute=@n26.then, Condition=@n26.bSuccess)
  ]}
 RenderConfirmDlg {nodes [
   n1 = Sequence()
   > n2 = Is Target 3L!(execute, self) -> (then, is_target_3l:bool)
   > n3 = Branch(Condition=@n2)
   n4 = Break Float Range(FloatRange=Get FrameRange())
   n5 = Break Float Range Bound(FloatRangeBound=@n4) -> (Value=~)
   n6 = Break Float Range()
   n7 = Break Float Range Bound(FloatRangeBound=@n6) -> (Value=~)
   n8 = Break Float Range Bound(FloatRangeBound=@n4.UpperBound) -> (Value=~)
   n9 = Break Float Range Bound(FloatRangeBound=@n6.UpperBound) -> (Value=~)
   > n10 = Branch(Condition=OR Boolean(A="Not Equal (Float)"(A=@n7.Value, B=@n5.Value), B="Not Equal (Float)"(A=@n9.Value, B=@n8.Value)))
   > Set render_confirm_dlg_msg(render_confirm_dlg_msg:Text="LOCTEXT(\"3L head rom range is [0-2731] but do you want to continue?\")") -> (Output_Get:Text)
   Set render_confirm_dlg_msg(execute=@n10.else, render_confirm_dlg_msg="LOCTEXT(\"Rendering \\\"3L\\\" head rom.\\r\\nDo you want to continue?\")")
   n11 = Break Float Range()
   n12 = Break Float Range Bound(FloatRangeBound=@n11) -> (Value=~)
   n13 = Break Float Range Bound(FloatRangeBound=@n11.UpperBound) -> (Value=~)
   n14 = Branch(execute=@n3.else, Condition=OR Boolean(A="Not Equal (Float)"(A=@n5.Value, B=@n12.Value), B="Not Equal (Float)"(A=@n8.Value, B=@n13.Value)))
   > Set render_confirm_dlg_msg(render_confirm_dlg_msg="LOCTEXT(\"Legacy head rom range is [0-1960] but do you want to continue?\")")
   Set render_confirm_dlg_msg(execute=@n14.else, render_confirm_dlg_msg="LOCTEXT(\"Rendering \\\"Legacy\\\" head rom.\\r\\nDo you want to continue?\")")
   n15 = Show Message Dialog(
    execute=@n1.then_1
    Title=Get render_confirm_dlg_title() -> (render_confirm_dlg_title:Text)
    Message=Get render_confirm_dlg_msg() -> (render_confirm_dlg_msg:Text)
    MessageType=YesNo
    DefaultValue=Yes
    MessageCategory=Info
   )
   > Return Node!(execute, continue:bool="Equal (Enum)"(A:EAppReturnType Enum=@n15, B:EAppReturnType Enum=Yes))
   RenderConfirmDlg() {is=title K2Node_FunctionEntry}
  ]}
]
variables [
 CharacterPartSM_Companion {type=FortShellMeshSkeletalMeshComponent, default=null}
 CID_Path {type=string, default=""}
 CIDAssetPreviewDefaultIconBrush {
  type=SlateBrush
  default {
   bIsDynamicallyLoaded=false
   drawAs=Image
   imageType=NoImage
   mirroring=NoMirror
   resourceName=None
   resourceObject=/Game/UI/Foundation/Modals/ErrorWindow/Materials/Icon-Llama-NoEyes
   tiling=NoTile
   imageSize=<65 65>
   margin=<ltrb 0 0 0 0>
   outlineSettings {
    bUseBrushTransparency=false
    roundingType=HalfHeightRadius
    width=0
    color {colorUseRule=UseColor_Specified, specifiedColor=<rgba 0 0 0 0>}
    cornerRadii=<0 0 0 1>
   }
   tintColor {colorUseRule=UseColor_Specified, specifiedColor=<rgba 1 1 1 1>}
   uVRegion {bIsValid=false, max=<0 0>, min=<0 0>}
  }
 }
 CIDAssetPreviewIconBrush {
  type=SlateBrush
  default {
   bIsDynamicallyLoaded=false
   drawAs=Image
   imageType=FullColor
   mirroring=NoMirror
   resourceName=None
   resourceObject=&Icon-Llama-NoEyes
   tiling=NoTile
   imageSize=<65 65>
   margin=<ltrb 0 0 0 0>
   outlineSettings {
    bUseBrushTransparency=false
    roundingType=HalfHeightRadius
    width=0
    color {colorUseRule=UseColor_Specified, specifiedColor=<rgba 0 0 0 0>}
    cornerRadii=<0 0 0 1>
   }
   tintColor {colorUseRule=UseColor_Specified, specifiedColor=<rgba 1 1 1 1>}
   uVRegion {bIsValid=false, max=<0 0>, min=<0 0>}
  }
 }
 Companion Anim Montage {type=AnimMontage, default=null}
 Companion Source CID {type=CosmeticCompanionItemDefinition, default=null}
 Companion Target CID {type=CosmeticCompanionItemDefinition, default=null}
 CompanionSM Component {type=SkeletalMeshComponent}
 Default_FrameRange {type=FloatRange, default {lowerBound {type=Exclusive, value=0}, upperBound {type=Exclusive, value=0}}}
 Error Title String {type=string, default=Error}
 FailureStatusIconBrush {
  type=SlateBrush
  default {
   bIsDynamicallyLoaded=false
   drawAs=Image
   imageType=NoImage
   mirroring=NoMirror
   resourceName=None
   resourceObject=null
   tiling=NoTile
   imageSize=<20 20>
   margin=<ltrb 0 0 0 0>
   outlineSettings {
    bUseBrushTransparency=false
    roundingType=HalfHeightRadius
    width=0
    color {colorUseRule=UseColor_Specified, specifiedColor=<rgba 0 0 0 0>}
    cornerRadii=<0 0 0 1>
   }
   tintColor {colorUseRule=UseColor_Specified, specifiedColor=<rgba 1 0 0 1>}
   uVRegion {bIsValid=false, max=<0 0>, min=<0 0>}
  }
 }
 Fort Companion Mannequins {type="array<FortCompanionMannequin>", default []}
 FrameRange {type=FloatRange, default {lowerBound {type=Exclusive, value=0}, upperBound {type=Exclusive, value=2731}}}
 gender {type=string, default=""}
 Head_CID {type=AthenaCharacterItemDefinition, default=null}
 headTypeText {type=string, default=None}
 Info Title String {type=string, default=Job Done}
 Invalid CID Message String {type=string, default="An Asset with a CID must be Selected!"}
 Level Error Message String {type=string, default="Open the FaceRom level!"}
 LevelSequence {type=LevelSequence, default=/Game/NeverCook/ArtTools/ROMs/FaceROM_Sequence}
 MRQSettings {type=MoviePipelinePrimaryConfig, default=&/FaceRom_MRQ_Settings}
 New Actor Skeletal Mesh Component {type=SkeletalMeshComponent}
 OutputDirectory {type=DirectoryPath, default {path=D:/FaceROMs}}
 Render Done Message String {type=string, default="Render is done!"}
 Resolution {type=CaptureResolution, default {resX=1920, resY=1080}}
 ROM {type=AnimMontage, default=null}
 Source {type=CosmeticCompanionItemDefinition, default=null}
 SuccessStatusIconBrush {
  type=SlateBrush
  default {
   bIsDynamicallyLoaded=false
   drawAs=Image
   imageType=NoImage
   mirroring=NoMirror
   resourceName=None
   resourceObject=null
   tiling=NoTile
   imageSize=<20 20>
   margin=<ltrb 0 0 0 0>
   outlineSettings {
    bUseBrushTransparency=false
    roundingType=HalfHeightRadius
    width=0
    color {colorUseRule=UseColor_Specified, specifiedColor=<rgba 0 0 0 0>}
    cornerRadii=<0 0 0 1>
   }
   tintColor {colorUseRule=UseColor_Specified, specifiedColor=<rgba 0 1 0 1>}
   uVRegion {bIsValid=false, max=<0 0>, min=<0 0>}
  }
 }
 Target {type=CosmeticCompanionItemDefinition, default=null}
 WorkingLevel {type=SoftWorldReference, default {worldAsset=/Game/NeverCook/ArtTools/ROMs/FaceROM}}
]
widgets {
 info {rootWidgetClass=&VerticalBox, widgetCount=37}
 widgets [
  MainVBox {widgetClassPath=&VerticalBox, widget=MainVBox}
  MainHBox {
   widgetClassPath=&HorizontalBox
   parent=MainVBox
   widget=MainHBox
   slot {class=VerticalBoxSlot, properties {padding=<ltrb 10 10 10 10>}}
  }
  AssetMainVBox {
   widgetClassPath=&VerticalBox
   parent=MainHBox
   widget=AssetMainVBox
   slot {class=HorizontalBoxSlot, properties {verticalAlignment=VAlign_Top, size {sizeRule=Fill, value=1}}}
  }
  Border {
   widgetClassPath=&Border
   parent=AssetMainVBox
   widget=Border
   properties {
    background {
     bIsDynamicallyLoaded=false
     drawAs=Image
     imageType=NoImage
     mirroring=NoMirror
     resourceName=None
     resourceObject=null
     tiling=NoTile
     imageSize=<32 32>
     margin=<ltrb 0 0 0 0>
     outlineSettings {
      bUseBrushTransparency=false
      roundingType=HalfHeightRadius
      width=0
      color {colorUseRule=UseColor_Specified, specifiedColor=<rgba 0 0 0 0>}
      cornerRadii=<0 0 0 0>
     }
     tintColor {colorUseRule=UseColor_Specified, specifiedColor=<rgba 0 0 0 1>}
     uVRegion {bIsValid=false, max=<0 0>, min=<0 0>}
    }
    padding=<ltrb 0 1 0 0>
   }
   slot {class=VerticalBoxSlot}
  }
  HorizontalBox_106 {
   widgetClassPath=&HorizontalBox
   parent=AssetMainVBox
   widget=HorizontalBox_106
   slot {class=VerticalBoxSlot, properties {padding=<ltrb 0 10 0 10>}}
  }
  Button_Open_Level {
   widgetClassPath=&EditorUtilityButton
   parent=HorizontalBox_106
   widget=Button_Open_Level
   properties {toolTipText=Open FaceRom Level}
   slot {class=HorizontalBoxSlot, properties {padding=<ltrb 0 0 5 0>}}
  }
  open_level_button_text {
   widgetClassPath=&TextBlock
   parent=Button_Open_Level
   widget=open_level_button_text
   properties {
    text=Open Level
    font {
     bForceMonospaced=false
     bMaterialIsStencil=false
     fontMaterial=null
     fontObject=&Roboto
     letterSpacing=0
     monospacedWidth=1
     size=13.5
     skewAmount=0
     typefaceFontName=Bold
     outlineSettings {
      bApplyOutlineToDropShadows=false
      bMiteredCorners=false
      bSeparateFillAlpha=false
      outlineMaterial=null
      outlineSize=0
      outlineColor=<rgba 0 0 0 1>
     }
    }
   }
   slot {class=ButtonSlot, properties {padding=<ltrb 4 5 4 5>}}
  }
  divider_1 {
   widgetClassPath=&Image
   parent=HorizontalBox_106
   widget=divider_1
   properties {brush {
     bIsDynamicallyLoaded=false
     drawAs=Image
     imageType=NoImage
     mirroring=NoMirror
     resourceName=None
     resourceObject=null
     tiling=NoTile
     imageSize=<1 28>
     margin=<ltrb 0 0 0 0>
     outlineSettings {
      bUseBrushTransparency=false
      roundingType=HalfHeightRadius
      width=0
      color {colorUseRule=UseColor_Specified, specifiedColor=<rgba 0 0 0 0>}
      cornerRadii=<0 0 0 0>
     }
     tintColor {colorUseRule=UseColor_Specified, specifiedColor=<rgba 0.0052 0.0052 0.0052 1>}
     uVRegion {bIsValid=false, max=<0 0>, min=<0 0>}
    }}
   slot {class=HorizontalBoxSlot, properties {verticalAlignment=VAlign_Center}}
  }
  divider {
   widgetClassPath=&Image
   parent=HorizontalBox_106
   widget=divider
   properties {brush {
     bIsDynamicallyLoaded=false
     drawAs=Image
     imageType=NoImage
     mirroring=NoMirror
     resourceName=None
     resourceObject=null
     tiling=NoTile
     imageSize=<1 28>
     margin=<ltrb 0 0 0 0>
     outlineSettings {
      bUseBrushTransparency=false
      roundingType=HalfHeightRadius
      width=0
      color {colorUseRule=UseColor_Specified, specifiedColor=<rgba 0 0 0 0>}
      cornerRadii=<0 0 0 0>
     }
     tintColor {colorUseRule=UseColor_Specified, specifiedColor=<rgba 0.0052 0.0052 0.0052 1>}
     uVRegion {bIsValid=false, max=<0 0>, min=<0 0>}
    }}
   slot {class=HorizontalBoxSlot, properties {verticalAlignment=VAlign_Center}}
  }
  head_type_label {
   widgetClassPath=&TextBlock
   parent=HorizontalBox_106
   widget=head_type_label
   properties {
    text=Head Type:
    font {
     bForceMonospaced=false
     bMaterialIsStencil=false
     fontMaterial=null
     fontObject=&Roboto
     letterSpacing=0
     monospacedWidth=1
     size=9
     skewAmount=0
     typefaceFontName=Regular
     outlineSettings {
      bApplyOutlineToDropShadows=false
      bMiteredCorners=false
      bSeparateFillAlpha=false
      outlineMaterial=null
      outlineSize=0
      outlineColor=<rgba 0 0 0 1>
     }
    }
   }
   slot {class=HorizontalBoxSlot, properties {horizontalAlignment=HAlign_Center, verticalAlignment=VAlign_Center, padding=<ltrb 30 0 15 0>}}
  }
  head_type_value {
   widgetClassPath=&TextBlock
   bIsVariable=true
   parent=HorizontalBox_106
   widget=head_type_value
   properties {
    text=None
    font {
     bForceMonospaced=false
     bMaterialIsStencil=false
     fontMaterial=null
     fontObject=&Roboto
     letterSpacing=0
     monospacedWidth=1
     size=9
     skewAmount=0
     typefaceFontName=Regular
     outlineSettings {
      bApplyOutlineToDropShadows=false
      bMiteredCorners=false
      bSeparateFillAlpha=false
      outlineMaterial=null
      outlineSize=0
      outlineColor=<rgba 0 0 0 1>
     }
    }
   }
   slot {class=HorizontalBoxSlot, properties {horizontalAlignment=HAlign_Center, verticalAlignment=VAlign_Center, padding=<ltrb 0 0 5 0>}}
  }
  companion_label {
   widgetClassPath=&TextBlock
   parent=HorizontalBox_106
   widget=companion_label
   properties {
    text=Companion
    font {
     bForceMonospaced=false
     bMaterialIsStencil=false
     fontMaterial=null
     fontObject=&Roboto
     letterSpacing=0
     monospacedWidth=1
     size=9
     skewAmount=0
     typefaceFontName=Regular
     outlineSettings {
      bApplyOutlineToDropShadows=false
      bMiteredCorners=false
      bSeparateFillAlpha=false
      outlineMaterial=null
      outlineSize=0
      outlineColor=<rgba 0 0 0 1>
     }
    }
   }
   slot {class=HorizontalBoxSlot, properties {horizontalAlignment=HAlign_Center, verticalAlignment=VAlign_Center, padding=<ltrb 30 0 15 0>}}
  }
  Companion_checkBox {
   widgetClassPath=&EditorUtilityCheckBox
   bIsVariable=true
   parent=HorizontalBox_106
   widget=Companion_checkBox
   slot {class=HorizontalBoxSlot}
  }
  Border_1 {
   widgetClassPath=&Border
   parent=AssetMainVBox
   widget=Border_1
   properties {
    background {
     bIsDynamicallyLoaded=false
     drawAs=Image
     imageType=NoImage
     mirroring=NoMirror
     resourceName=None
     resourceObject=null
     tiling=NoTile
     imageSize=<32 32>
     margin=<ltrb 0 0 0 0>
     outlineSettings {
      bUseBrushTransparency=false
      roundingType=HalfHeightRadius
      width=0
      color {colorUseRule=UseColor_Specified, specifiedColor=<rgba 0 0 0 0>}
      cornerRadii=<0 0 0 0>
     }
     tintColor {colorUseRule=UseColor_Specified, specifiedColor=<rgba 0 0 0 1>}
     uVRegion {bIsValid=false, max=<0 0>, min=<0 0>}
    }
    padding=<ltrb 0 1 0 0>
   }
   slot {class=VerticalBoxSlot}
  }
  HorizontalBox_304 {
   widgetClassPath=&HorizontalBox
   parent=AssetMainVBox
   widget=HorizontalBox_304
   slot {class=VerticalBoxSlot, properties {padding=<ltrb 4 2 4 10>}}
  }
  Asset_Main_Layout {
   widgetClassPath=&HorizontalBox
   bIsVariable=true
   parent=HorizontalBox_304
   widget=Asset_Main_Layout
   slot {class=HorizontalBoxSlot, properties {size {sizeRule=Fill, value=1}}}
  }
  HorizontalBox_242 {
   widgetClassPath=&HorizontalBox
   parent=Asset_Main_Layout
   widget=HorizontalBox_242
   slot {class=HorizontalBoxSlot, properties {horizontalAlignment=HAlign_Left, verticalAlignment=VAlign_Center}}
  }
  AssetPreviewIcon {
   widgetClassPath=&Image
   bIsVariable=true
   parent=HorizontalBox_242
   widget=AssetPreviewIcon
   properties {
    toolTipText=FN CID Character Preview Icon
    brush {
     bIsDynamicallyLoaded=false
     drawAs=Image
     imageType=NoImage
     mirroring=NoMirror
     resourceName=None
     resourceObject=null
     tiling=NoTile
     imageSize=<128 128>
     margin=<ltrb 0 0 0 0>
     outlineSettings {
      bUseBrushTransparency=false
      roundingType=HalfHeightRadius
      width=0
      color {colorUseRule=UseColor_Specified, specifiedColor=<rgba 0 0 0 0>}
      cornerRadii=<0 0 0 0>
     }
     tintColor {colorUseRule=UseColor_Specified, specifiedColor=<rgba 1 1 1 1>}
     uVRegion {bIsValid=false, max=<0 0>, min=<0 0>}
    }
   }
   slot {class=HorizontalBoxSlot, properties {padding=<ltrb 0 0 10 0>}}
  }
  AssetsDetailsView {
   widgetClassPath=&DetailsView
   parent=Asset_Main_Layout
   widget=AssetsDetailsView
   properties {columnWidth=1, categoriesToShow [Assets]}
   slot {class=HorizontalBoxSlot, properties {size {sizeRule=Fill, value=1}}}
  }
  HorizontalBox_0 {
   widgetClassPath=&HorizontalBox
   parent=Asset_Main_Layout
   widget=HorizontalBox_0
   slot {class=HorizontalBoxSlot, properties {size {sizeRule=Fill, value=1}}}
  }
  CompanionTargetCID {
   widgetClassPath=&DetailsView
   parent=HorizontalBox_0
   widget=CompanionTargetCID
   properties {visibility=Collapsed, propertiesToShow [Target]}
   slot {class=HorizontalBoxSlot, properties {size {sizeRule=Fill, value=1}}}
  }
  CompanionSourceCID {
   widgetClassPath=&DetailsView
   parent=HorizontalBox_0
   widget=CompanionSourceCID
   properties {visibility=Collapsed, propertiesToShow [Source]}
   slot {class=HorizontalBoxSlot, properties {size {sizeRule=Fill, value=1}}}
  }
  CompanionFaceROM {
   widgetClassPath=&DetailsView
   parent=HorizontalBox_0
   widget=CompanionFaceROM
   properties {visibility=Collapsed, propertiesToShow [ROM]}
   slot {class=HorizontalBoxSlot, properties {size {sizeRule=Fill, value=1}}}
  }
  OpenInSGButton {
   widgetClassPath=&EditorUtilityButton
   parent=Asset_Main_Layout
   widget=OpenInSGButton
   properties {
    toolTipText=Open Asset in ShotGrid Browser
    widgetStyle {
     clickedSlateSound {resourceObject=null}
     disabled {
      bIsDynamicallyLoaded=false
      drawAs=RoundedBox
      imageType=NoImage
      mirroring=NoMirror
      resourceName=None
      resourceObject=null
      tiling=NoTile
      imageSize=<0 0>
      margin=<ltrb 0 0 0 0>
      outlineSettings {
       bUseBrushTransparency=false
       roundingType=FixedRadius
       width=1
       color {colorUseRule=UseColor_ColorTable, specifiedColor=<rgba 0 0 0 0>}
       cornerRadii=<4 4 4 4>
      }
      tintColor {colorUseRule=UseColor_Specified, specifiedColor=<rgba 0.0395 0.0395 0.0395 1>}
      uVRegion {bIsValid=false, max=<0 0>, min=<0 0>}
     }
     disabledForeground {colorUseRule=UseColor_ColorTable, specifiedColor=<rgba 0 0 0 0>}
     hovered {
      bIsDynamicallyLoaded=false
      drawAs=RoundedBox
      imageType=NoImage
      mirroring=NoMirror
      resourceName=None
      resourceObject=null
      tiling=NoTile
      imageSize=<32 32>
      margin=<ltrb 0 0 0 0>
      outlineSettings {
       bUseBrushTransparency=false
       roundingType=FixedRadius
       width=1
       color {colorUseRule=UseColor_ColorTable, specifiedColor=<rgba 0 0 0 0>}
       cornerRadii=<4 4 4 4>
      }
      tintColor {colorUseRule=UseColor_Specified, specifiedColor=<rgba 0.0953 0.0953 0.0953 1>}
      uVRegion {bIsValid=false, max=<0 0>, min=<0 0>}
     }
     hoveredForeground {colorUseRule=UseColor_ColorTable, specifiedColor=<rgba 0 0 0 0>}
     hoveredSlateSound {resourceObject=null}
     normal {
      bIsDynamicallyLoaded=false
      drawAs=NoDrawType
      imageType=NoImage
      mirroring=NoMirror
      resourceName=None
      resourceObject=null
      tiling=NoTile
      imageSize=<32 32>
      margin=<ltrb 0 0 0 0>
      outlineSettings {
       bUseBrushTransparency=false
       roundingType=FixedRadius
       width=1
       color {colorUseRule=UseColor_ColorTable, specifiedColor=<rgba 0 0 0 0>}
       cornerRadii=<4 4 4 4>
      }
      tintColor {colorUseRule=UseColor_Specified, specifiedColor=<rgba 0.0048 0.0048 0.0048 0>}
      uVRegion {bIsValid=false, max=<0 0>, min=<0 0>}
     }
     normalForeground {colorUseRule=UseColor_ColorTable, specifiedColor=<rgba 0 0 0 0>}
     normalPadding=<ltrb 0 1.5 0 1.5>
     pressed {
      bIsDynamicallyLoaded=false
      drawAs=RoundedBox
      imageType=NoImage
      mirroring=NoMirror
      resourceName=None
      resourceObject=null
      tiling=NoTile
      imageSize=<32 32>
      margin=<ltrb 0 0 0 0>
      outlineSettings {
       bUseBrushTransparency=false
       roundingType=FixedRadius
       width=1
       color {colorUseRule=UseColor_ColorTable, specifiedColor=<rgba 0 0 0 0>}
       cornerRadii=<4 4 4 4>
      }
      tintColor {colorUseRule=UseColor_Specified, specifiedColor=<rgba 0.0284 0.0284 0.0284 1>}
      uVRegion {bIsValid=false, max=<0 0>, min=<0 0>}
     }
     pressedForeground {colorUseRule=UseColor_ColorTable, specifiedColor=<rgba 0 0 0 0>}
     pressedPadding=<ltrb 0 1.5 0 1.5>
     pressedSlateSound {resourceObject=null}
    }
   }
   slot {class=HorizontalBoxSlot, properties {verticalAlignment=VAlign_Top, padding=<ltrb 4 4 4 4>}}
  }
  OpenInSGIcon {
   widgetClassPath=&Image
   bIsVariable=true
   parent=OpenInSGButton
   widget=OpenInSGIcon
   properties {
    toolTipText=Open in SG
    brush {
     bIsDynamicallyLoaded=false
     drawAs=Image
     imageType=FullColor
     mirroring=NoMirror
     resourceName=None
     resourceObject=/FortUILibrary/Foundation/Buttons/Textures/T_Icon_ExternalLink
     tiling=NoTile
     imageSize=<20 20>
     margin=<ltrb 0 0 0 0>
     outlineSettings {
      bUseBrushTransparency=false
      roundingType=HalfHeightRadius
      width=0
      color {colorUseRule=UseColor_Specified, specifiedColor=<rgba 0 0 0 0>}
      cornerRadii=<0 0 0 0>
     }
     tintColor {colorUseRule=UseColor_Specified, specifiedColor=<rgba 1 1 1 1>}
     uVRegion {bIsValid=false, max=<0 0>, min=<0 0>}
    }
   }
   slot {class=ButtonSlot, properties {verticalAlignment=VAlign_Top}}
  }
  Border_2 {
   widgetClassPath=&Border
   parent=AssetMainVBox
   widget=Border_2
   properties {
    background {
     bIsDynamicallyLoaded=false
     drawAs=Image
     imageType=NoImage
     mirroring=NoMirror
     resourceName=None
     resourceObject=null
     tiling=NoTile
     imageSize=<32 32>
     margin=<ltrb 0 0 0 0>
     outlineSettings {
      bUseBrushTransparency=false
      roundingType=HalfHeightRadius
      width=0
      color {colorUseRule=UseColor_Specified, specifiedColor=<rgba 0 0 0 0>}
      cornerRadii=<0 0 0 0>
     }
     tintColor {colorUseRule=UseColor_Specified, specifiedColor=<rgba 0 0 0 1>}
     uVRegion {bIsValid=false, max=<0 0>, min=<0 0>}
    }
    padding=<ltrb 0 1 0 0>
   }
   slot {class=VerticalBoxSlot}
  }
  VerticalBox_0 {
   widgetClassPath=&VerticalBox
   parent=MainVBox
   widget=VerticalBox_0
   properties {visibility=Visible}
   slot {class=VerticalBoxSlot}
  }
  HorizontalBox_115 {
   widgetClassPath=&HorizontalBox
   parent=VerticalBox_0
   widget=HorizontalBox_115
   properties {visibility=Visible}
   slot {class=VerticalBoxSlot}
  }
  Render_DetailsView {
   widgetClassPath=&DetailsView
   parent=HorizontalBox_115
   widget=Render_DetailsView
   properties {bShowScrollBar=false, columnWidth=0.55, categoriesToShow [Render Options]}
   slot {class=HorizontalBoxSlot, properties {size {sizeRule=Fill, value=1}}}
  }
  HorizontalBox {
   widgetClassPath=&HorizontalBox
   parent=MainVBox
   widget=HorizontalBox
   slot {class=VerticalBoxSlot, properties {padding=<ltrb 4 2 4 2>}}
  }
  HorizontalBox_1 {
   widgetClassPath=&HorizontalBox
   parent=HorizontalBox
   widget=HorizontalBox_1
   properties {toolTipText="Render the final ROM video and Upload to the ShotGrid Task, Mark as Ready for Review"}
   slot {class=HorizontalBoxSlot, properties {verticalAlignment=VAlign_Center, size {sizeRule=Fill, value=1}}}
  }
  Button_Reset_Render_option {
   widgetClassPath=&EditorUtilityButton
   parent=HorizontalBox_1
   widget=Button_Reset_Render_option
   properties {toolTipText=reset render settings to default}
   slot {class=HorizontalBoxSlot, properties {verticalAlignment=VAlign_Center, padding=<ltrb 0 20 0 20>, size {sizeRule=Fill, value=1}}}
  }
  reset_render_option {
   widgetClassPath=&TextBlock
   parent=Button_Reset_Render_option
   widget=reset_render_option
   properties {
    text=Reset
    font {
     bForceMonospaced=false
     bMaterialIsStencil=false
     fontMaterial=null
     fontObject=&Roboto
     letterSpacing=0
     monospacedWidth=1
     size=13.5
     skewAmount=0
     typefaceFontName=Bold
     outlineSettings {
      bApplyOutlineToDropShadows=false
      bMiteredCorners=false
      bSeparateFillAlpha=false
      outlineMaterial=null
      outlineSize=0
      outlineColor=<rgba 0 0 0 1>
     }
    }
   }
   slot {class=ButtonSlot}
  }
  Button_Render_ROM {
   widgetClassPath=&EditorUtilityButton
   parent=HorizontalBox_1
   widget=Button_Render_ROM
   properties {toolTipText=render movie}
   slot {class=HorizontalBoxSlot, properties {verticalAlignment=VAlign_Center, padding=<ltrb 0 20 0 20>, size {sizeRule=Fill, value=1}}}
  }
  Render {
   widgetClassPath=&TextBlock
   parent=Button_Render_ROM
   widget=Render
   properties {
    text=Render
    font {
     bForceMonospaced=false
     bMaterialIsStencil=false
     fontMaterial=null
     fontObject=&Roboto
     letterSpacing=0
     monospacedWidth=1
     size=13.5
     skewAmount=0
     typefaceFontName=Bold
     outlineSettings {
      bApplyOutlineToDropShadows=false
      bMiteredCorners=false
      bSeparateFillAlpha=false
      outlineMaterial=null
      outlineSize=0
      outlineColor=<rgba 0 0 0 1>
     }
    }
   }
   slot {class=ButtonSlot}
  }
  Button_Upload_ROM {
   widgetClassPath=&EditorUtilityButton
   parent=HorizontalBox_1
   widget=Button_Upload_ROM
   properties {toolTipText=upload the rendered movie to SG}
   slot {class=HorizontalBoxSlot, properties {verticalAlignment=VAlign_Center, padding=<ltrb 0 20 0 20>, size {sizeRule=Fill, value=1}}}
  }
  Upload {
   widgetClassPath=&TextBlock
   parent=Button_Upload_ROM
   widget=Upload
   properties {
    text=Upload to SG
    font {
     bForceMonospaced=false
     bMaterialIsStencil=false
     fontMaterial=null
     fontObject=&Roboto
     letterSpacing=0
     monospacedWidth=1
     size=13.5
     skewAmount=0
     typefaceFontName=Bold
     outlineSettings {
      bApplyOutlineToDropShadows=false
      bMiteredCorners=false
      bSeparateFillAlpha=false
      outlineMaterial=null
      outlineSize=0
      outlineColor=<rgba 0 0 0 1>
     }
    }
   }
   slot {class=ButtonSlot}
  }
 ]
}
