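# Editor utility widget that fixes a picked Skeletal Mesh whose object name differs from its package name.
# A custom asset picker filtered to Skeletal Mesh feeds the Rename button, which compares the asset name with the last segment of its package path.
# On a mismatch, it rebuilds the path in the same folder from the asset name and calls EditorAssetSubsystem's RenameLoadedAsset, printing the outcome.
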EditorUtilityWidgetBlueprint /Game/NeverCook/ArtTools/Tech_Anim/Fix_Hidden_Mesh
graphs [EventGraph {nodes [
   "On Clicked (EditorUtilityButton_43)"() {is=title K2Node_ComponentBoundEvent}
   > n1 = LoadAndGetAssetObject!(
    execute
    self:EUW Custom Asset Picker Object Reference=Get EUW_CustomAssetPicker() -> (EUW_CustomAssetPicker:EUW Custom Asset Picker Object Reference)
   ) -> (then, asset:Object Reference, HasSelection:bool)
   > Set package_display_name(package_display_name:String=GetDisplayName(Object=GetPackageForObject!(Object:Object Reference=@n1) -> (ReturnValue:Package Object Reference))) -> (Output_Get:String)
   > Set display_name(display_name:String=GetDisplayName(Object=@n1)) -> (Output_Get:String)
   n2 = int + int(A=FindSubstring(SearchIn=Get package_display_name() -> (package_display_name:String), Substring=/, bSearchFromEnd=true))
   n3 = Get package_display_name()
   > n4 = Branch(Condition="Not Equal, Case Insenstive (String)"!(A:String=Get display_name() -> (display_name:String), B:String=RightChop!(SourceString:String=@n3, Count:int=@n2) -> (ReturnValue:String)) -> (ReturnValue:bool))
   n5 = Get display_name()
   n6 = Append!(A=Left!(SourceString:String=@n3, Count:int=@n2) -> (ReturnValue:String), B=@n5, C:String=., D:String=@n5) -> (ReturnValue)
   > n7 = RenameLoadedAsset!(
    execute
    self:Editor Asset Subsystem Object Reference=Get EditorAssetSubsystem() -> (ReturnValue:Editor Asset Subsystem Object Reference) {is=title K2Node_GetEditorSubsystem}
    SourceAsset:Object Reference=@n1
    DestinationAssetPath:String=@n6
   ) -> (then, ReturnValue:bool)
   > n8 = Branch(Condition=@n7)
   > PrintString(InString=Append!(A="Renamed Skeletal Mesh: ", B=@n6) -> (ReturnValue))
   PrintString(execute=@n8.else, InString=Rename Failed)
   PrintString(execute=@n4.else, InString=Nothing to rename)
   Event PreConstruct()
   Event Construct()
   Event Tick()
  ]}]
variables [
 AssetPicker {type=EUW_CustomAssetPicker_C, default=null}
 display_name {type=string, default=""}
 package_display_name {type=string, default=""}
]
widgets {
 info {rootWidgetClass=&StackBox, widgetCount=6}
 widgets [
  StackBox_0 {widgetClassPath=&StackBox, widget=StackBox_0, properties {orientation=Orient_Vertical}}
  StackBox_51 {widgetClassPath=&StackBox, parent=StackBox_0, widget=StackBox_51, slot {class=StackBoxSlot}}
  EUW_CustomAssetPicker {
   widgetClassPath=/Game/NeverCook/EditorUtilities/CustomAssetPicker/EUW_CustomAssetPicker
   bIsVariable=true
   parent=StackBox_51
   widget=EUW_CustomAssetPicker
   properties {label=Pick an asset to rename, classFilters [&SkeletalMesh]}
   slot {class=StackBoxSlot}
  }
  StackBox_107 {widgetClassPath=&StackBox, parent=StackBox_0, widget=StackBox_107, slot {class=StackBoxSlot}}
  EditorUtilityButton_43 {widgetClassPath=&EditorUtilityButton, parent=StackBox_107, widget=EditorUtilityButton_43, slot {class=StackBoxSlot}}
  TextBlock_126 {
   widgetClassPath=&TextBlock
   parent=EditorUtilityButton_43
   widget=TextBlock_126
   properties {text=Rename}
   slot {class=ButtonSlot}
  }
 ]
}
